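# Niagara sprite system for the idle effect of the one-handed female Celestial pickaxe: a pulsing glow with stars.
# Heart bursts one purple sprite per 1.8-second loop, with a double-peaked alpha curve over its slightly longer lifetime.
# StarsCentral fills a small Y-axis cylinder with fading blue stars using randomized dynamic material parameters.
# StarsCentral001 intermittently emits larger violet-blue stars that drift outward, inherit some velocity and fall, all emitters using camera offset.
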NiagaraSystem /BRCosmetics/Weapons/FORT_Melee/Pickaxe_Celestial_Female_1h/FX/NS_Celestial_F_Idle
dependencies {usedDataInterfaces [&NiagaraDataInterfaceCurve], usedDynamicInputs [], usedModules [&SystemState]}
emitters {
 Heart {
  emitterUpdateScript [
   &EmitterState(Inactive Response=::NewEnumerator1, Life Cycle Mode=::NewEnumerator1, Loop Duration=1.8, UseLoopDelay=false)
   &SpawnBurst_Instantaneous(Spawn Count=1, Use Spawn Probability=false)
  ]
  particleSpawnScript [
   /Niagara/Modules/Spawn/Initialization/InitializeParticle(
    Color=<rgba 0.7487 0.3875 1.5 0.4>
    Lifetime=2
    Module.Write Color=true
    Module.Write Lifetime=true
    Module.Write Mass=true
    Module.Write Material Random=true
    Module.Write Position=false
    Module.Write Scale=false
    Module.Write SpriteRotation=false
    Module.Write SpriteSize=true
    Position=&SimulationPosition()
    Sprite Size=&Vector2DFromFloat(Value=&UniformRangedFloat(Minimum=15, Maximum=20))
   )
   &CameraOffset(Camera Offset Amount=30)
  ]
  particleUpdateScript [
   &ParticleState()
   &SolveForcesAndVelocity(Clamp Velocity=false, Limit Acceleration=false)
   &ScaleColor(
    Scale Alpha=&FloatFromCurve(FloatCurve=DataInterface() {propertyValues {Curve {keys [
         <at 0.0051 0.1919 RCIM_Cubic>
         {arriveTangent=15.763, interpMode=RCIM_Cubic, leaveTangent=15.7629, tangentMode=RCTM_User, time=0.1, value=0.85}
         {arriveTangent=-6.6929, interpMode=RCIM_Cubic, leaveTangent=-6.6929, tangentMode=RCTM_User, time=0.15, value=1.2}
         {interpMode=RCIM_Cubic, tangentMode=RCTM_User, time=0.3, value=0.65}
         {arriveTangent=1.1254, interpMode=RCIM_Cubic, leaveTangent=1.1254, tangentMode=RCTM_User, time=0.81, value=0.8}
         {arriveTangent=8.7241, interpMode=RCIM_Cubic, leaveTangent=8.7241, tangentMode=RCTM_User, time=0.91, value=1.15}
         <at 1 0.2 RCIM_Cubic>
        ]}}})
    ScaleA=true
    ScaleRGB=true
   )
  ]
  rendererClasses [&NiagaraSpriteRendererProperties]
  renderers [{rendererClass=&NiagaraSpriteRendererProperties}]
 }
 StarsCentral {
  emitterUpdateScript [
   &EmitterState(Life Cycle Mode=::NewEnumerator1, Loop Duration=1, UseLoopDelay=false)
   &SpawnRate(Use Spawn Probability=false)
  ]
  particleSpawnScript [
   &InitializeParticle(
    Color=&UniformRangedLinearColor(Minimum=<rgba 0.1703 0.4623 1 0.7>, Maximum=<rgba 0.1666 0.9901 1 0.7>)
    Lifetime=&UniformRangedFloat(Minimum=2, Maximum=3)
    Module.Write Color=true
    Module.Write Lifetime=true
    Module.Write Mass=true
    Module.Write Material Random=true
    Module.Write Position=false
    Module.Write Scale=false
    Module.Write SpriteRotation=false
    Module.Write SpriteSize=true
    Position=&SimulationPosition()
    Sprite Size=&Vector2DFromFloat(Value=&UniformRangedFloat(Minimum=2, Maximum=5))
   )
   &CameraOffset(Camera Offset Amount=28)
   &DynamicMaterialParameters(
    Index 0 Param 1=&UniformRangedFloat(Minimum=0.4, Maximum=0.7)
    Index 0 Param 2=&UniformRangedFloat(Minimum=0.8, Maximum=1.55)
    Index 0 Param 3=1
    Index 0 Param 4=1
    Module.Param0WriteEnabled=true
    Module.Param1WriteEnabled=true
    Module.Param2WriteEnabled=true
    Module.Param3WriteEnabled=true
   )
   &CylinderLocation(
    Cylinder Height=8
    Cylinder Radius=2
    MaskSpawnGroup=false
    Module.Enable Lathe Profile=false
    Module.Override Randomness=false
    Module.Surface Only=false
    Orientation Axis:ENiagaraOrientationAxis=::YAxis
   )
  ]
  particleUpdateScript [
   &ParticleState()
   &SolveForcesAndVelocity(Clamp Velocity=false, Limit Acceleration=false)
   &ScaleColor(
    Scale Alpha=&FloatFromCurve(FloatCurve=DataInterface() {propertyValues {Curve {keys [
         <at 0 0 RCIM_Cubic>
         {interpMode=RCIM_Cubic, tangentMode=RCTM_User, time=0.2, value=1}
         {interpMode=RCIM_Cubic, tangentMode=RCTM_User, time=0.8, value=1}
         <at 1 0 RCIM_Cubic>
        ]}}})
    ScaleA=true
    ScaleRGB=true
   )
  ]
  rendererClasses [&NiagaraSpriteRendererProperties]
  renderers [{rendererClass=&NiagaraSpriteRendererProperties}]
 }
 StarsCentral001 {
  emitterUpdateScript [
   &EmitterState(
    Life Cycle Mode=::NewEnumerator1
    Loop Delay=&UniformRangedFloat(Minimum=0, Maximum=3)
    Loop Duration=&UniformRangedFloat(Minimum=0.1, Maximum=1)
    Recalculate Duration Each Loop=true
    UseLoopDelay=true
   )
   &SpawnRate(SpawnRate=&UniformRangedFloat(Minimum=1, Maximum=5), Use Spawn Probability=false)
  ]
  particleSpawnScript [
   &InitializeParticle(
    Color=&UniformRangedLinearColor(Minimum=<rgba 0.358 0.0562 1 1>, Maximum=<rgba 0.0677 0.3228 1 1>)
    Lifetime=&UniformRangedFloat(Minimum=2, Maximum=3)
    Module.Write Color=true
    Module.Write Lifetime=true
    Module.Write Mass=true
    Module.Write Material Random=true
    Module.Write Position=false
    Module.Write Scale=false
    Module.Write SpriteRotation=false
    Module.Write SpriteSize=true
    Position=&SimulationPosition()
    Sprite Size=&Vector2DFromFloat(Value=&UniformRangedFloat(Minimum=5, Maximum=10))
   )
   &CameraOffset(Camera Offset Amount=28)
   &DynamicMaterialParameters(
    Index 0 Param 1=&UniformRangedFloat(Minimum=0.4, Maximum=0.7)
    Index 0 Param 2=&UniformRangedFloat(Minimum=0.8, Maximum=1.55)
    Index 0 Param 3=1
    Index 0 Param 4=1
    Module.Param0WriteEnabled=true
    Module.Param1WriteEnabled=true
    Module.Param2WriteEnabled=true
    Module.Param3WriteEnabled=true
   )
   &CylinderLocation(
    Cylinder Height=8
    Cylinder Radius=2
    MaskSpawnGroup=false
    Module.Enable Lathe Profile=false
    Module.Override Randomness=false
    Module.Surface Only=false
    Orientation Axis:ENiagaraOrientationAxis=::YAxis
   )
   &AddVelocity(Velocity=&Multiply_Vector(A=$Output.CylinderLocation.CylinderNormal, B=&VectorFromFloat(Value=1)))
   &InheritVelocity(Module.Limit Speed=true, Velocity Scale=<0.15 0.15 0.15>)
  ]
  particleUpdateScript [
   &ParticleState()
   &AccelerationForce(Acceleration=&UniformRangedVector(Minimum=<0 0 -100>, Maximum=<0 0 -50>))
   &SolveForcesAndVelocity(Clamp Velocity=false, Limit Acceleration=false)
   &ScaleColor(
    Scale Alpha=&FloatFromCurve(FloatCurve=DataInterface() {propertyValues {Curve {keys [
         <at 0 0 RCIM_Cubic>
         {interpMode=RCIM_Cubic, tangentMode=RCTM_User, time=0.2, value=1}
         {interpMode=RCIM_Cubic, tangentMode=RCTM_User, time=0.8, value=1}
         <at 1 0 RCIM_Cubic>
        ]}}})
    ScaleA=true
    ScaleRGB=true
   )
  ]
  rendererClasses [&NiagaraSpriteRendererProperties]
  renderers [{rendererClass=&NiagaraSpriteRendererProperties}]
 }
}
summary {
 systemName=NS_Celestial_F_Idle
 emitters [
  {emitterName=Heart, rendererClasses [&NiagaraSpriteRendererProperties]}
  {emitterName=StarsCentral, rendererClasses [&NiagaraSpriteRendererProperties]}
  {emitterName=StarsCentral001, rendererClasses [&NiagaraSpriteRendererProperties]}
 ]
 userVariables []
}
