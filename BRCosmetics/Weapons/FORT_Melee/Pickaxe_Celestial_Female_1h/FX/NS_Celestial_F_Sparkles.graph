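# Niagara sparkle effect for the Celestial female one-handed pickaxe cosmetic, drawing small purple-to-blue star sprites.
# A single emitter, StarsCentral001, spawns short-lived particles at a random rate inside a small Y-axis cylinder, with randomized material parameters and a camera offset.
# Particles drift along the cylinder normal and accelerate downward.
# A multi-peak size curve makes them twinkle, while an alpha curve fades them in and out.
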NiagaraSystem /BRCosmetics/Weapons/FORT_Melee/Pickaxe_Celestial_Female_1h/FX/NS_Celestial_F_Sparkles
dependencies {usedDataInterfaces [&NiagaraDataInterfaceCurve], usedDynamicInputs [], usedModules [&SystemState]}
emitters {StarsCentral001 {
  emitterUpdateScript [
   &EmitterState(Life Cycle Mode=::NewEnumerator1, Loop Behavior=::NewEnumerator1, Loop Duration=0.5, UseLoopDelay=false)
   &SpawnRate(SpawnRate=&UniformRangedFloat(Minimum=10, Maximum=25), Use Spawn Probability=false)
  ]
  particleSpawnScript [
   /Niagara/Modules/Spawn/Initialization/InitializeParticle(
    Color=&UniformRangedLinearColor(Minimum=<rgba 0.358 0.0562 1 1>, Maximum=<rgba 0.0677 0.3228 1 1>)
    Lifetime=&UniformRangedFloat(Minimum=0.5, Maximum=1.5)
    Module.Write Color=true
    Module.Write Lifetime=true
    Module.Write Mass=true
    Module.Write Material Random=true
    Module.Write Position=false
    Module.Write Scale=false
    Module.Write SpriteRotation=false
    Module.Write SpriteSize=true
    Position=&SimulationPosition()
    Sprite Size=&Vector2DFromFloat(Value=&UniformRangedFloat(Minimum=5, Maximum=10))
   )
   &CameraOffset(Camera Offset Amount=28)
   &DynamicMaterialParameters(
    Index 0 Param 1=&UniformRangedFloat(Minimum=0.4, Maximum=0.7)
    Index 0 Param 2=&UniformRangedFloat(Minimum=0.8, Maximum=1.55)
    Index 0 Param 3=1
    Index 0 Param 4=1
    Module.Param0WriteEnabled=true
    Module.Param1WriteEnabled=true
    Module.Param2WriteEnabled=true
    Module.Param3WriteEnabled=true
   )
   &CylinderLocation(
    Cylinder Height=8
    Cylinder Radius=2
    MaskSpawnGroup=false
    Module.Enable Lathe Profile=false
    Module.Override Randomness=false
    Module.Surface Only=false
    Orientation Axis:ENiagaraOrientationAxis=::YAxis
   )
   &AddVelocity(Velocity=&Multiply_Vector(
     A=$Output.CylinderLocation.CylinderNormal
     B=&VectorFromFloat(Value=&UniformRangedFloat(Minimum=1, Maximum=5))
    ))
  ]
  particleUpdateScript [
   &ParticleState()
   &AccelerationForce(Acceleration=<0 0 -50>)
   &SolveForcesAndVelocity(Clamp Velocity=false, Limit Acceleration=false)
   &ScaleColor(
    Scale Alpha=&FloatFromCurve(FloatCurve=DataInterface() {propertyValues {Curve {keys [
         <at 0 0 RCIM_Cubic>
         {interpMode=RCIM_Cubic, tangentMode=RCTM_User, time=0.2, value=1}
         {interpMode=RCIM_Cubic, tangentMode=RCTM_User, time=0.8, value=1}
         <at 1 0 RCIM_Cubic>
        ]}}})
    ScaleA=true
    ScaleRGB=true
   )
   &ScaleSpriteSize(Scale Factor=&Vector2DFromFloat(Value=&FloatFromCurve(FloatCurve=DataInterface() {propertyValues {Curve {keys [
          <at 0 0 RCIM_Cubic>
          {arriveTangent=1.3333, interpMode=RCIM_Cubic, leaveTangent=1.3333, time=0.05, value=1.25}
          {arriveTangent=-1.2, interpMode=RCIM_Cubic, leaveTangent=-1.2, time=0.15, value=0.2}
          {arriveTangent=0.24, interpMode=RCIM_Cubic, leaveTangent=0.24, time=0.3, value=0.95}
          {arriveTangent=0.1667, interpMode=RCIM_Cubic, leaveTangent=0.1667, time=0.4, value=0.26}
          {arriveTangent=-0.3143, interpMode=RCIM_Cubic, leaveTangent=-0.3143, time=0.6, value=1}
          {arriveTangent=-1.6, interpMode=RCIM_Cubic, leaveTangent=-1.6, time=0.75, value=0.15}
          {arriveTangent=-0.6, interpMode=RCIM_Cubic, leaveTangent=-0.6, time=0.8, value=0.68}
          <at 1 0 RCIM_Cubic>
         ]}}})))
  ]
  rendererClasses [&NiagaraSpriteRendererProperties]
  renderers [{rendererClass=&NiagaraSpriteRendererProperties}]
 }}
summary {
 systemName=NS_Celestial_F_Sparkles
 emitters [{emitterName=StarsCentral001, rendererClasses [&NiagaraSpriteRendererProperties]}]
 userVariables []
}
