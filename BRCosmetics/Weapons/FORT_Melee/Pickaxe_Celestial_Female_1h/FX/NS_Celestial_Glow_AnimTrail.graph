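# Niagara animation-trail effect for the Celestial Female one-handed pickaxe, drawing short glowing ribbon trails from bone or socket positions on a skeletal mesh user parameter.
# Two near-identical Nebula emitters drive dynamic material parameters with curves over life; one sets a fixed ribbon facing.
# A narrower, shorter-lived orange SharpEdge emitter adds a camera offset.
# All three share AnimTrail spawn, initialize and update modules, boosted color and curve-based alpha fades.
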NiagaraSystem /BRCosmetics/Weapons/FORT_Melee/Pickaxe_Celestial_Female_1h/FX/NS_Celestial_Glow_AnimTrail
dependencies {usedDataInterfaces [&NiagaraDataInterfaceCurve], usedDynamicInputs [], usedModules [&SystemState]}
emitters {
 AnimTrail_Nebula {
  emitterUpdateScript [
   &EmitterState(Life Cycle Mode=::NewEnumerator1, Loop Behavior=::NewEnumerator1, Loop Duration=0.22, UseLoopDelay=false)
   /Game/Effects/Niagara/Modules/AnimTrail/FN_SpawnAnimTrail/FN_SpawnAnimTrail(Number of Anim Trails=1, SpawnRate=45)
  ]
  particleSpawnScript [
   /Game/Effects/Niagara/Modules/AnimTrail/FN_InitializeAnimTrail/FN_InitializeAnimTrailBoneIndecies(Second Bone or Socket Index=1)
   /Game/Effects/Niagara/Modules/AnimTrail/FN_InitializeAnimTrail/FN_InitializeAnimTrail(AnimTrailWidth=1, Color=<rgba 2.8867 1.3 3 2>, Lifetime=0.4, Write Material Random=false)
  ]
  particleUpdateScript [
   &SolveForcesAndVelocity(Clamp Velocity=false, Limit Acceleration=false)
   /Game/Effects/Niagara/Modules/AnimTrail/FN_UpdateAnimTrail/FN_UpdateAnimTrail()
   &DynamicMaterialParameters(
    Index 0 Param 1=&FloatFromCurve(FloatCurve=DataInterface() {propertyValues {Curve {keys [<at 0 1 RCIM_Cubic>, {arriveTangent=-1.6667, interpMode=RCIM_Cubic, leaveTangent=-1.6667, time=0.4, value=0.8}, <at 0.6 0 RCIM_Cubic>]}}})
    Index 0 Param 2=&FloatFromCurve(FloatCurve=DataInterface() {propertyValues {Curve {keys [
         {arriveTangent=0.7648, interpMode=RCIM_Cubic, leaveTangent=0.7648, tangentMode=RCTM_User, time=0.0523, value=0.0099}
         {arriveTangent=-0.0129, interpMode=RCIM_Cubic, leaveTangent=-0.0129, time=0.5, value=0.4}
         {arriveTangent=-1.1763, interpMode=RCIM_Cubic, leaveTangent=-1.1763, tangentMode=RCTM_User, time=0.82, value=0}
        ]}}})
    Index 0 Param 3=1
    Index 0 Param 4=1
    Module.Param0WriteEnabled=true
    Module.Param1WriteEnabled=true
    Module.Param2WriteEnabled=true
    Module.Param3WriteEnabled=true
   )
   &ScaleColor(
    Scale Alpha=&FloatFromCurve(FloatCurve=DataInterface() {propertyValues {Curve {keys [
         <at 0 0 RCIM_Cubic>
         {arriveTangent=2.3333, interpMode=RCIM_Cubic, leaveTangent=2.3333, time=0.04, value=0.95}
         {arriveTangent=-0.2679, interpMode=RCIM_Cubic, leaveTangent=-0.2679, time=0.3, value=0.7}
         {arriveTangent=-1, interpMode=RCIM_Cubic, leaveTangent=-1, time=0.6, value=0.8}
         <at 1 0 RCIM_Cubic>
        ]}}})
    Scale RGB=<10 10 10>
    ScaleA=true
    ScaleRGB=true
   )
   &ScaleRibbonWidth(Ribbon Width Scale=&FloatFromCurve(FloatCurve=DataInterface() {propertyValues {Curve {keys [<at 0 1 RCIM_Cubic>, {arriveTangent=-1.6223, interpMode=RCIM_Cubic, leaveTangent=-1.6223, tangentMode=RCTM_User, time=1, value=0}]}}})) {enabled=false}
   /SetVariables() {isSetParameters=true, module=SetVariables_#1}
  ]
  rendererClasses [&NiagaraRibbonRendererProperties]
  renderers [{rendererClass=&NiagaraRibbonRendererProperties}]
 }
 AnimTrail_Nebula001 {
  emitterUpdateScript [
   &EmitterState(Life Cycle Mode=::NewEnumerator1, Loop Behavior=::NewEnumerator1, Loop Duration=0.22, UseLoopDelay=false)
   &FN_SpawnAnimTrail(Number of Anim Trails=1, SpawnRate=45)
  ]
  particleSpawnScript [
   &FN_InitializeAnimTrailBoneIndecies(Second Bone or Socket Index=1)
   &FN_InitializeAnimTrail(AnimTrailWidth=1, Color=<rgba 2.6317 1.3 3 2>, Lifetime=0.4, Write Material Random=false)
  ]
  particleUpdateScript [
   &SolveForcesAndVelocity(Clamp Velocity=false, Limit Acceleration=false)
   &FN_UpdateAnimTrail()
   &DynamicMaterialParameters(
    Index 0 Param 1=&FloatFromCurve(FloatCurve=DataInterface() {propertyValues {Curve {keys [<at 0 1 RCIM_Cubic>, {arriveTangent=-1.6667, interpMode=RCIM_Cubic, leaveTangent=-1.6667, time=0.4, value=0.8}, <at 0.6 0 RCIM_Cubic>]}}})
    Index 0 Param 2=&FloatFromCurve(FloatCurve=DataInterface() {propertyValues {Curve {keys [
         {arriveTangent=0.7648, interpMode=RCIM_Cubic, leaveTangent=0.7648, tangentMode=RCTM_User, time=0.0523, value=0.0099}
         {arriveTangent=-0.0129, interpMode=RCIM_Cubic, leaveTangent=-0.0129, time=0.5, value=0.4}
         {arriveTangent=-1.1763, interpMode=RCIM_Cubic, leaveTangent=-1.1763, tangentMode=RCTM_User, time=0.82, value=0}
        ]}}})
    Index 0 Param 3=1
    Index 0 Param 4=1
    Module.Param0WriteEnabled=true
    Module.Param1WriteEnabled=true
    Module.Param2WriteEnabled=true
    Module.Param3WriteEnabled=true
   )
   &ScaleColor(
    Scale Alpha=&FloatFromCurve(FloatCurve=DataInterface() {propertyValues {Curve {keys [
         <at 0 0 RCIM_Cubic>
         {arriveTangent=2.3333, interpMode=RCIM_Cubic, leaveTangent=2.3333, time=0.04, value=0.95}
         {arriveTangent=-0.2679, interpMode=RCIM_Cubic, leaveTangent=-0.2679, time=0.3, value=0.7}
         {arriveTangent=-1, interpMode=RCIM_Cubic, leaveTangent=-1, time=0.6, value=0.8}
         <at 1 0 RCIM_Cubic>
        ]}}})
    Scale RGB=<10 10 10>
    ScaleA=true
    ScaleRGB=true
   )
   &ScaleRibbonWidth(Ribbon Width Scale=&FloatFromCurve(FloatCurve=DataInterface() {propertyValues {Curve {keys [<at 0 1 RCIM_Cubic>, {arriveTangent=-1.6223, interpMode=RCIM_Cubic, leaveTangent=-1.6223, tangentMode=RCTM_User, time=1, value=0}]}}})) {enabled=false}
   /SetVariables(Particles.RibbonFacing=<0 1 0>) {isSetParameters=true, module=SetVariables_#1}
  ]
  rendererClasses [&NiagaraRibbonRendererProperties]
  renderers [{rendererClass=&NiagaraRibbonRendererProperties}]
 }
 AnimTrail_SharpEdge {
  emitterUpdateScript [
   &EmitterState(Life Cycle Mode=::NewEnumerator1, Loop Behavior=::NewEnumerator1, Loop Duration=0.22, UseLoopDelay=false)
   &FN_SpawnAnimTrail(Number of Anim Trails=1, SpawnRate=45)
  ]
  particleSpawnScript [
   &FN_InitializeAnimTrailBoneIndecies(Second Bone or Socket Index=1)
   &FN_InitializeAnimTrail(AnimTrailWidth=0.6, Color=<rgba 3 0.4771 0.1439 1>, Lifetime=0.25, Write Material Random=false)
   &CameraOffset(Camera Offset Amount=5)
  ]
  particleUpdateScript [
   &SolveForcesAndVelocity(Clamp Velocity=false, Limit Acceleration=false)
   &FN_UpdateAnimTrail()
   &ScaleColor(
    Scale Alpha=&FloatFromCurve(FloatCurve=DataInterface() {propertyValues {Curve {keys [
         <at 0 0 RCIM_Cubic>
         {arriveTangent=2.3333, interpMode=RCIM_Cubic, leaveTangent=2.3333, time=0.04, value=0.95}
         {arriveTangent=-0.2679, interpMode=RCIM_Cubic, leaveTangent=-0.2679, time=0.3, value=0.7}
         {arriveTangent=-1, interpMode=RCIM_Cubic, leaveTangent=-1, time=0.6, value=0.8}
         <at 1 0 RCIM_Cubic>
        ]}}})
    Scale RGB=<10 10 10>
    ScaleA=true
    ScaleRGB=true
   )
   &ScaleRibbonWidth(Ribbon Width Scale=&FloatFromCurve(FloatCurve=DataInterface() {propertyValues {Curve {keys [<at 0 1 RCIM_Cubic>, {arriveTangent=-1.6223, interpMode=RCIM_Cubic, leaveTangent=-1.6223, tangentMode=RCTM_User, time=1, value=0}]}}})) {enabled=false}
   /SetVariables(Particles.RibbonFacing=<1 0 0>) {enabled=false, isSetParameters=true, module=SetVariables_#1}
  ]
  rendererClasses [&NiagaraRibbonRendererProperties]
  renderers [{rendererClass=&NiagaraRibbonRendererProperties}]
 }
}
summary {
 systemName=NS_Celestial_Glow_AnimTrail
 emitters [
  {emitterName=AnimTrail_Nebula, rendererClasses [&NiagaraRibbonRendererProperties]}
  {emitterName=AnimTrail_SharpEdge, rendererClasses [&NiagaraRibbonRendererProperties]}
  {emitterName=AnimTrail_Nebula001, rendererClasses [&NiagaraRibbonRendererProperties]}
 ]
 userVariables [
  User.RibbonIDOffset {type=NiagaraInt32}
  User.SkeletalMesh {
   type=NiagaraDataInterfaceSkeletalMesh
   defaultValue {
    struct=NiagaraExt_VariableValue_DataInterface
    value {dataInterface=&self:NiagaraDataInterfaceSkeletalMesh_0}
   }
  }
 ]
}
